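# Translucent, unlit, two-sided smoke material for Niagara mesh particles in the launch pad item's launch effect, by name for the sides of a smoke disk.
# A panning, tiled puff texture is sharpened and multiplied by a static disk texture.
# Dynamic parameter Param1 blends from plain disk to this detailed version in color and opacity.
# Particle color tints both, with day/night bloom control on emissive and depth and view-angle fades on opacity.
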Material /LaunchPadItemGameplay/FX/Launch/Materials/M_Smoke_Disk_Sides
graph {
 properties {
  bUsedWithNanite=true
  bUsedWithNiagaraMeshParticles=true
  blendMode=BLEND_Translucent
  pixelDepthOffsetMode=PDOM_Legacy
  shadingModel=MSM_Unlit
  twoSided=true
  usedShadingModels=Unlit
 }
 nodes [
  n1 = ParticleColor()
  n2 = TextureSample() {texture=/Game/Effects/Fort_Effects/Textures/Smoke/T_Smoke_Bomb_Disk}
  n3 = TextureCoordinate() {uTiling=10, vTiling=2}
  n4 = Power(Base=TextureSample(UVs=Panner(Coordinate=@n3) {speedX=0.02, speedY=0.25}) {texture=/Game/Global/Textures/Default/Tileable_Patterns/T_Smoke_Puff_Tile_Up}.RGB)
  n5 = Multiply(A=Multiply(A=Clamp(None=@n4), B=@n2.RGB)) {constB=2}
  n6 = DynamicParameter() {defaultValue=<rgba 0 0 0 1>}
  n7 = LinearInterpolate(A=@n2.RGB, B=@n5, Alpha=@n6.Param1)
  MaterialFunctionCall(Emissive=MaterialFunctionCall!(DayBloom_Mult=Constant() {r=1}, In=Multiply(A=@n1.RGB, B=@n7)) {materialFunction=/Game/Packages/Fortnite_UniqueMaterials_1/MH_Materials/DayNight_Auto_Bloom_Control}.Result) {materialFunction=&MF_FX_EmissiveAdjustments}
  n8 = LinearInterpolate(A=@n2.A, B=@n5, Alpha=@n6.Param1)
  Multiply(
   A=DepthFade(Opacity=Multiply(A=@n1.A, B=@n8)) {fadeDistanceDefault=30}
   B=MaterialFunctionCall("FadeSensitivity (optional)"=Constant() {r=3}) {materialFunction=&FadeBasedOnViewAngle}.Result
  )
 ]
}
